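# Niagara system for the prop spawn effect of the PolkaSkate Box emote; only the BlowingParticles emitter carries modules and renderers here.
# Two delayed bursts use spawn groups to separate randomly launched, small brown debris from a second group, with a visibility tag feeding the mesh and decal renderers.
# Debris falls under gravity and shrinks over its lifetime, bouncing via collision gated by platform quality or User.bIsFrontEndPreview.
# An alternate analytical-plane collision module is present but disabled.
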NiagaraSystem /BRCosmetics/Animation/Game/MainPlayer/Emotes/PolkaSkate_Box/FX/NS_Emote_PolkaSkate_Box_PropSpawn
dependencies {
 usedDataInterfaces [&NiagaraDataInterfacePlatformSet, &NiagaraDataInterfaceVectorCurve, &NiagaraDataInterfaceCurve]
 usedDynamicInputs []
 usedModules [&SystemState]
}
emitters {
 AirglowLWE {rendererClasses [], renderers []}
 AirglowLWE001 {rendererClasses [], renderers []}
 BlowingParticles {
  emitterUpdateScript [
   &EmitterState(Life Cycle Mode=::NewEnumerator1, Loop Behavior=::NewEnumerator1, Loop Delay=0.1, Loop Duration=4, UseLoopDelay=true)
   &SpawnBurst_Instantaneous(
    Spawn Count=&RandomRangeInt(Minimum=9, Maximum=15, Recalculate Random Each Loop=false)
    Use Spawn Probability=false
    UseLoopCountLimit=false
   )
   &SpawnBurst_Instantaneous(Spawn Group=1, Use Spawn Probability=false, UseLoopCountLimit=false) {module=SpawnBurst_Instantaneous001}
  ]
  particleSpawnScript [
   /SetVariables(Particles.VisibilityTag=&SpawnGroup()) {isSetParameters=true}
   &InitializeParticle(
    Color=<rgba 0.151 0.075 0.0368 1>
    Lifetime=&MaskFloatBySpawnGroup(Masked Float=&RandomRangeFloat(Minimum=0.9, Maximum=2.9377), Passthrough Float=0.4984)
    Mass Max=1.5
    Mass Mode=::NewEnumerator3
    Mesh Scale Mode=::NewEnumerator1
    Mesh Uniform Scale Max=0.1
    Mesh Uniform Scale Min=0.05
    Sprite Rotation Mode=::NewEnumerator2
    UsePositionOffset=true
    Write Material Random=true
   )
   &ShapeLocation(
    MaskSpawnGroup=false
    Offset=<1 -28 12>
    Offset Mode=::NewEnumerator0
    Override Seed=false
    Ring Radius=12
    Shape Primitive=::NewEnumerator4
    UseOwnerScale=true
   )
   &InitialMeshOrientation(Use Rotation Vector=false)
   &AddVelocity(
    Cone Angle=&RandomRangeFloat(Minimum=40.8464, Maximum=186.7016)
    Module.UseConeFalloff=false
    Module.UseVelDistribution=false
    Override Seed=false
    Rotation Angle=180
    Rotation Mode=::NewEnumerator1
    Velocity Mode=::NewEnumerator2
    Velocity Speed=&MaskFloatBySpawnGroup(Masked Float=&RandomRangeFloat(Minimum=50, Maximum=250), Passthrough Float=0)
   )
  ]
  particleUpdateScript [
   &GravityForce(Gravity=<0 0 -1180.0895>)
   &Collision(
    Age Colliding Particles=false
    Collision Enabled=&Boolean_OrOperation(
     Bool1=&IsPlatformSetActive(New Platform Set=DataInterface() {propertyValues {Platforms {qualityLevelMask=-8}}})
     Bool2=&Boolean_NotOperation(Bool=$User.bIsFrontEndPreview)
    )
    Friction=0.5
    Friction During a Bounce=0.3
    KillOnCollision=false
    Module.Control Rest State Via Surface Normal=false
    Module.EnableMaxCollisionCount=false
    Module.Randomize Collision Normal Vector=false
    Radius Calculation Type=::NewEnumerator1
    Restitution=0.44
   )
   &Collision(
    Age Colliding Particles=false
    Analytical Collision Plane Position 1:Vector3f=<0 0 -50>
    Analytical Collision Plane Space:ENiagaraCoordinateSpace=::Local
    CPU Collision Type=::NewEnumerator3
    Collision Enabled=&Boolean_OrOperation(
     Bool1=&IsPlatformSetActive(New Platform Set=DataInterface() {propertyValues {Platforms {qualityLevelMask=-28}}})
     Bool2=$User.bIsFrontEndPreview
    )
    KillOnCollision=false
    Module.Control Rest State Via Surface Normal=false
    Module.EnableMaxCollisionCount=false
    Module.Randomize Collision Normal Vector=false
    Perform 2 Analytical Collision Queries=false
   ) {enabled=false, module=Collision001}
   &SolveForcesAndVelocity(Clamp Velocity=false, Limit Acceleration=false)
   &ParticleState()
   &ScaleMeshSize(Scale Factor=&VectorFromCurve(VectorCurve=DataInterface() {propertyValues {XCurve {keys [<at 0 1>,<at 1 0>]}, YCurve {keys [<at 0 1>,<at 1 0>]}, ZCurve {keys [<at 0 1>,<at 1 0>]}}}))
   &DynamicMaterialParameters(Module.Index0Float0Write=true, Module.Index0Float1Write=true, Module.Index0Float2Write=true, Module.Index0Float3Write=true)
   /SetVariables(Particles.float:NiagaraFloat=&FloatFromCurve(
     FloatCurve=DataInterface() {propertyValues {Curve {
        defaultValue=1
        keys [
         {arriveTangent=4.1785, interpMode=RCIM_Cubic, leaveTangent=4.1785, tangentMode=RCTM_User, time=0, value=0}
         <at 0.2318 0.8879 RCIM_Cubic>
         {arriveTangent=0.0224, interpMode=RCIM_Cubic, leaveTangent=0.0224, tangentMode=RCTM_User, time=0.9929, value=0}
        ]
       }}}
     CurveIndex=$Particles.Age
    )) {isSetParameters=true}
  ]
  rendererClasses [&NiagaraMeshRendererProperties, &NiagaraDecalRendererProperties]
  renderers [
   {rendererClass=&NiagaraMeshRendererProperties, rendererIndex=0}
   {rendererClass=&NiagaraDecalRendererProperties, rendererIndex=1}
  ]
 }
 Empty {rendererClasses [], renderers []}
 Pebbles {rendererClasses [], renderers []}
 SmokeLWE {rendererClasses [], renderers []}
 SmokeLWE001 {rendererClasses [], renderers []}
 SmokeLWE002 {rendererClasses [], renderers []}
 SparklesLWE {rendererClasses [], renderers []}
 SparklesLWE001 {rendererClasses [], renderers []}
}
summary {
 systemName=NS_Emote_PolkaSkate_Box_PropSpawn
 emitters [
  {emitterName=AirglowLWE, rendererClasses []}
  {emitterName=SmokeLWE, rendererClasses []}
  {emitterName=SparklesLWE, rendererClasses []}
  {emitterName=SmokeLWE001, rendererClasses []}
  {emitterName=SmokeLWE002, rendererClasses []}
  {emitterName=Pebbles, rendererClasses []}
  {emitterName=AirglowLWE001, rendererClasses []}
  {emitterName=SparklesLWE001, rendererClasses []}
  {
   emitterName=BlowingParticles
   rendererClasses [&NiagaraMeshRendererProperties, &NiagaraDecalRendererProperties]
  }
  {emitterName=Empty, rendererClasses []}
 ]
 userVariables [User.bIsFrontEndPreview {type=NiagaraBool}]
}
